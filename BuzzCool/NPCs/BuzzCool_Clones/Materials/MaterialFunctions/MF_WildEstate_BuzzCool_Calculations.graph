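# Material function that computes anime-style lighting gradients and masks for the BuzzCool clone NPC materials.
# Normals blend toward vertex-color baked normals, then combine with a light-direction function and cinematic override to form NdotL.
# A backlit term, overlay-blended threshold and banded cel-shading function shape shading, while a sphere-masked Phong term with Z-lock and camera-facing blend forms specular.
# It unpacks sheen, threshold and thickness from the STT input, alongside an FX map.
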MaterialFunction /BuzzCool/NPCs/BuzzCool_Clones/Materials/MaterialFunctions/MF_WildEstate_BuzzCool_Calculations
graph {nodes [
  FunctionOutput(None=NamedRerouteUsage()) {outputName=Specular_Gradient, sortPriority=5}
  FunctionOutput(None=NamedRerouteUsage()) {outputName=Sheen_Mask, sortPriority=6}
  FunctionOutput(None=NamedRerouteUsage()) {outputName=FX, sortPriority=9}
  FunctionOutput(None=NamedRerouteUsage()) {outputName=Thickness, sortPriority=7}
  FunctionOutput(None=NamedRerouteUsage()) {outputName=Shading_Gradient, sortPriority=1}
  FunctionOutput(None=NamedRerouteUsage()) {outputName=Threshold_Gradient, sortPriority=2}
  n1 = NamedRerouteUsage()
  n2 = ScalarParameter() {group=02_Anime Parameters, parameterName=ShadingCurve_Bias, sliderMax=1, sliderMin=-1}
  n3 = MaterialFunctionCall!(
   NdL=Saturate(None=Add(A=@n1, B=@n2))
   Width=ScalarParameter() {defaultValue=0.02, group=02_Anime Parameters, parameterName=ShadingCurve_Width, sliderMax=1}
   Bands Amount=ScalarParameter() {defaultValue=1, group=02_Anime Parameters, parameterName=ShadingCurve_Bands, sliderMax=4, sliderMin=1}
  ) {materialFunction=/Game/Global/Materials/Character_Parent_2023/Anime_Parent_2023/MaterialFunctions/MF_SmoothCelShadingWithBands}
  n4 = ScalarParameter() {
   defaultValue=1
   desc="To convert the (mathmatical) linear greadient to linear perceived gradient"
   group=09_Anime Parameters_TechArt
   parameterName=ShadingCurve_Exp
   sliderMax=2.2
   sliderMin=0.4545
  }
  NamedRerouteDeclaration(None=Saturate(None=Power(Base=@n3.Divide, Exp=@n4) {constExponent=2.2})) {name=Shading_Gradient, nodeColor=<rgba 1 0.1875 0 1>}
  n5 = Transform(None=NamedRerouteUsage())
  n6 = VertexColor()
  n7 = Transform(None=Subtract(A=Multiply(A=@n6) {constB=2})) {transformSourceType=TRANSFORMSOURCE_Local}
  n8 = ScalarParameter() {desc="0.0 = surface normals, 1.0 baked vertex color normal", group=02_Anime Parameters, parameterName=VertexBakedNormal_Blend, sliderMax=1}
  n9 = Multiply(A=@n8, B=@n6.A)
  NamedRerouteDeclaration(None=Normalize(VectorInput=LinearInterpolate(A=@n5, B=@n7, Alpha=@n9))) {name=Normal_WS, nodeColor=<rgba 0.2083 0.2083 1 1>}
  NamedRerouteDeclaration(None=FunctionInput() {inputName=FX, inputType=FunctionInput_Vector4, sortPriority=4}) {name=Map_FX, nodeColor=<rgba 0 0.326 1 1>}
  n10 = MaterialFunctionCall(Float3=FunctionInput() {inputName=STT, sortPriority=1, previewValue=<0 0.5 0 0>}) {materialFunction=&BreakOutFloat3Components}
  NamedRerouteDeclaration(None=@n10.B) {name=MapInput_Thickness, nodeColor=<rgba 0 1 0.4064 1>}
  NamedRerouteDeclaration(None=@n10.R) {name=MapInput_Sheen, nodeColor=<rgba 0 0.3016 1 1>}
  n11 = Constant2Vector() {g=1, r=1}
  n12 = ScalarParameter() {
   desc="LockedZ only for the specular highlight.\r\nfor a streched highlight in on the Z axis use the Blinn highlight"
   group=02_Anime Parameters
   parameterName=SpecularHighlight_LockedZVector
   sliderMax=1
  }
  n13 = Saturate(None=OneMinus(None=@n12))
  n14 = Normalize(VectorInput=Multiply(A=NamedRerouteUsage(), B=AppendVector(A=@n11, B=@n13)))
  n15 = ReflectionVectorWS(CustomWorldNormal=NamedRerouteUsage())
  n16 = Saturate(None=LinearInterpolate(
    A=DotProduct(A=@n14, B=@n15) {desc=Phong}
    B=NamedRerouteUsage()
    Alpha=ScalarParameter() {group=02_Anime Parameters, parameterName=SpecularHighlight_CameraFacing_Blend, sliderMax=1}
   ))
  n17 = Constant() {r=1}
  n18 = ScalarParameter() {defaultValue=0.1, group=02_Anime Parameters, parameterName=SpecularHighlight_Size, sliderMax=1, sliderMin=0.0001}
  n19 = ScalarParameter() {defaultValue=0.2, group=02_Anime Parameters, parameterName=SpecularHighlight_Hardness, sliderMax=1}
  NamedRerouteDeclaration(None=Multiply(
    A=SphereMask(A=@n16, B=@n17, Radius=@n18, Hardness=@n19) {hardnessPercent=78}
    B=ScalarParameter() {defaultValue=1, group=02_Anime Parameters, parameterName=SpecularHighlight_Intensity, sliderMax=8}
   )) {name=Specular_Curve, nodeColor=<rgba 1 0.4453 0 1>}
  NamedRerouteDeclaration(None=DotProduct(A=NamedRerouteUsage(), B=CameraVectorWS())) {name=NdotV, nodeColor=<rgba 1 0.6242 0 1>}
  NamedRerouteDeclaration(None=@n10.G) {name=MapInput_Theshold, nodeColor=<rgba 1 0.4584 0 1>}
  n20 = ScalarParameter() {group=02_Anime Parameters, parameterName=Threshold_Min, sliderMax=1}
  n21 = ScalarParameter() {defaultValue=1, group=02_Anime Parameters, parameterName=Threshold_Max, sliderMax=1}
  n22 = Power(Base=NamedRerouteUsage(), Exp=ScalarParameter() {defaultValue=1, group=02_Anime Parameters, parameterName=Threshold_Exp, sliderMax=8})
  NamedRerouteDeclaration(None=Saturate(None=LinearInterpolate(A=@n20, B=@n21, Alpha=@n22))) {name=Threshold, nodeColor=<rgba 1 0.2971 0.8877 1>}
  NamedRerouteDeclaration(None=MaterialFunctionCall() {materialFunction=/Game/Global/Materials/Character_Parent_2023/Anime_Parent_2023/MaterialFunctions/MFC_AnimeShader_LightDirection}.Direction Vector) {name=LightVector, nodeColor=<rgba 1 0.934 0.5103 1>}
  n23 = NamedRerouteUsage()
  n24 = NamedRerouteUsage()
  NamedRerouteDeclaration(None=MaterialFunctionCall!(In=DotProduct(A=@n23, B=@n24)) {materialFunction=/Game/Packages/Fortress_SharedMaterials/Cinematics/Anime/MF_CINE_AnimeLightingFeatures}.CINE_ShadingCurve) {name=NdotL, nodeColor=<rgba 0.9723 1 0.2825 1>}
  NamedRerouteDeclaration(None=FunctionInput() {inputName=Normal_Map, sortPriority=2, previewValue=<0 0 1 1>}) {name=Normal, nodeColor=<rgba 0 0.4109 1 1>}
  n25 = NamedRerouteUsage()
  n26 = NamedRerouteUsage()
  n27 = Add(A=ScalarParameter() {defaultValue=4, group=02_Anime Parameters, parameterName=BackLit_Exponent, sliderMax=12, sortPriority=31}) {constB=2.2}
  n28 = Saturate(None=SmoothStep(Value=Power(Base=@n26, Exp=@n27) {constExponent=4}) {constMax=0, constMin=0.02})
  n29 = Add(
   A=DotProduct(
    A=MaterialFunctionCall() {materialFunction=&CameraDirectionVector}.Result
    B=NamedRerouteUsage()
   )
   B=ScalarParameter() {defaultValue=0.25, group=02_Anime Parameters, parameterName=BackLit_Bias, sliderMax=1}
  )
  n30 = ScalarParameter() {defaultValue=0.75, group=02_Anime Parameters, parameterName=BackLit_MaxBias, sliderMax=1}
  n31 = LinearInterpolate(
   A=@n25
   B=Add(A=@n25, B=NamedRerouteDeclaration(None=@n28) {name=BackLit_Mask, nodeColor=<rgba 0 1 0.5078 1>})
   Alpha=NamedRerouteDeclaration(None=Clamp(None=@n29, Max=@n30)) {name=BackLit_Curve, nodeColor=<rgba 0.3207 0 1 1>}
  )
  n32 = Add(A=Multiply(A=NamedRerouteDeclaration(None=@n31) {name=BackLitShading_-1to1, nodeColor=<rgba 0.1307 1 0.334 1>}) {constB=0.5}) {constB=0.5}
  NamedRerouteDeclaration(None=MaterialFunctionCall!(Base=NamedRerouteUsage(), Blend=NamedRerouteDeclaration(None=@n32) {name=BackLitShading_0to1, nodeColor=<rgba 0 1 0.9234 1>}) {materialFunction=/Game/Global/Materials/Character_Parent_2023/Anime_Parent_2023/MaterialFunctions/MF_BlendOverlay}.Result) {name=BackLitShading+Threshold, nodeColor=<rgba 0 1 0.9234 1>}
 ]}
