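# Material function for BuzzCool clone NPCs that builds final material attributes for anime shading with an optional PBR override.
# The anime branch uses the Cloth shading model with black base color, packing inputs into SubsurfaceColor and other attributes.
# The PBR branch tints LUT-applied color by an FX-channel mask, with an optional lit/shaded color mix and metallic bias.
# UsePBRSwitch enables blending the branches by PBR_Shading; otherwise only anime attributes are output.
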MaterialFunction /BuzzCool/NPCs/BuzzCool_Clones/Materials/MaterialFunctions/MF_WildEstate_BuzzCool_Post
graph {nodes [
  n1 = ChannelMaskParameter(None=NamedRerouteUsage()) {group=03_Anime_PBR_Override, maskChannel=Alpha, parameterName=FXChannel_PBR_GMap, sortPriority=10, defaultValue=<rgba 0 0 0 1>}
  n2 = OneMinus(None=@n1.R)
  n3 = ScalarParameter() {group=03_Anime_PBR_Override, parameterName=PBR_GMap_Intensity, sliderMax=1}
  n4 = NamedRerouteUsage()
  n5 = Constant()
  n6 = SetMaterialAttributes!(
   MaterialAttributes
   BaseColor=Multiply(
    A=LinearInterpolate(B=@n2, Alpha=@n3) {constA=1}
    B=MaterialFunctionCall!(In=@n4, LUT, LUT Multiplier) {materialFunction=/Game/Packages/Fortress_SharedMaterials/Anime_Master_Material/MF_ApplyLUT}.Result
   )
   Specular=NamedRerouteUsage()
   Roughness=NamedRerouteUsage()
   Metallic=NamedRerouteUsage()
   AmbientOcclusion=Constant() {r=1}
   Opacity=@n5
   SubsurfaceColor=@n5
   CustomData0=@n5
   Normal=NamedRerouteUsage()
  ) {attributeSetTypes [
    #1
    #2
    #3
    #4
    #5
    #6
    #7
    #8
    #9
   ]}
  n7 = NamedRerouteDeclaration(None=Saturate(None=ScalarParameter() {group=03_Anime_PBR_Override, parameterName=PBR_Shading, sliderMax=1, sortPriority=5})) {name=PBR_Switch, nodeColor=<rgba 0 1 0.8566 1>}
  n8 = Constant()
  n9 = NamedRerouteUsage()
  n10 = SetMaterialAttributes!(
   MaterialAttributes
   BaseColor=@n8
   Metallic=@n9
   Specular=@n8
   Roughness=Constant() {r=1}
   SubsurfaceColor=MaterialFunctionCall(X=NamedRerouteUsage(), Y=@n9, Z=NamedRerouteUsage()) {materialFunction=&MakeFloat3}.Result
   CustomData0=@n8
   Normal=NamedRerouteUsage()
   AmbientOcclusion=OneMinus(None=NamedRerouteUsage())
   ShadingModel=ShadingModel() {shadingModel=MSM_Cloth}
  ) {attributeSetTypes [
    #1
    #4
    #2
    #3
    #7
    #8
    #9
    #5
    #10
   ]}
  FunctionOutput(None=StaticSwitchParameter(True=BlendMaterialAttributes(A=@n10, B=@n6, Alpha=@n7), False=@n10) {group=02_Anime Parameters, parameterName=UsePBRSwitch, sortPriority=0})
  NamedRerouteDeclaration(None=FunctionInput() {inputName=Color_Shaded, sortPriority=32, previewValue=<0.18 0.18 0.18 1>}) {name=BC_Shaded, nodeColor=<rgba 0 0.326 1 1>}
  NamedRerouteDeclaration(None=FunctionInput() {inputName=Thickness, inputType=FunctionInput_Scalar, sortPriority=7, previewValue=<1 1 1 1>}) {name=Thickness, nodeColor=<rgba 0.4766 0 1 1>}
  NamedRerouteDeclaration(None=FunctionInput() {inputName=NormalBuffer, sortPriority=12, previewValue=<0.5 0.5 1 1>}) {name=NormalBuffer, nodeColor=<rgba 0.2083 0.2083 1 1>}
  NamedRerouteDeclaration(None=FunctionInput() {inputName=Sheen_Mask, inputType=FunctionInput_Scalar, sortPriority=6}) {name=Sheen_Mask, nodeColor=<rgba 0.401 0.8071 1 1>}
  n11 = NamedRerouteUsage()
  n12 = LinearInterpolate(
   A=NamedRerouteUsage()
   B=@n11
   Alpha=ScalarParameter() {defaultValue=0.5, group=03_Anime_PBR_Override, parameterName=PBR_BaseColor_Shaded_Blend, sliderMax=1}
  )
  n13 = Add(A=@n12, B=ScalarParameter() {group=03_Anime_PBR_Override, parameterName=PBR_BaseColor_Metallic_Bias, sliderMax=1})
  n14 = NamedRerouteUsage()
  NamedRerouteDeclaration(None=StaticSwitchParameter(True=LinearInterpolate(A=@n12, B=@n13, Alpha=@n14), False=@n11) {group=03_Anime_PBR_Override, parameterName=UseLitShadedColorMix}) {name=PBR_BaseColor, nodeColor=<rgba 0.125 1 0 1>}
  n15 = FunctionInput() {inputName=Threshold_Gradient, inputType=FunctionInput_Scalar, sortPriority=2, previewValue=<0.5 0 0 1>}
  NamedRerouteDeclaration(None=@n15) {name=Threshold_Curve, nodeColor=<rgba 0.7973 1 0 1>}
  n16 = MaterialFunctionCall(Float3=FunctionInput() {inputName=SRM, sortPriority=33}) {materialFunction=&BreakOutFloat3Components}
  NamedRerouteDeclaration(None=@n16.R) {name=MapInput_PBR_Specular, nodeColor=<rgba 1 0.0765 0 1>}
  NamedRerouteDeclaration(None=@n16.G) {name=MapInput_PBR_Roughness, nodeColor=<rgba 0 1 0.0508 1>}
  NamedRerouteDeclaration(None=@n16.B) {name=MapInput_PBR_Metallic, nodeColor=<rgba 0.0365 0 1 1>}
  NamedRerouteDeclaration(None=FunctionInput() {inputName=FX, inputType=FunctionInput_Vector4, sortPriority=9}) {name=FX, nodeColor=<rgba 1 0 0.4219 1>}
  NamedRerouteDeclaration(None=FunctionInput() {inputName=Color_Lit, sortPriority=31, previewValue=<0.9 0.9 0.9 1>}) {name=BC_Lit, nodeColor=<rgba 0 0.326 1 1>}
 ]}
