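# BuzzCool clone material function that adds an optional stylized, banded sun-lighting look to incoming Material Attributes.
# Quantizes vertex normal dot sunlight direction (from FortniteMaterialParameters, optionally modified) into clamped, remapped steps, then Overlay-blends that onto BaseColor.
# An optional earlier stage bands normals by lerping them toward the camera vector.
# Material proxies get the unmodified input; several experimental branches, including diffuse quantization, never reach the output.
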MaterialFunction /BuzzCool/NPCs/BuzzCool_Clones/Materials/MaterialFunctions/M_WildEstate_BuzzCool_QantizeColor
graph {nodes [
  n1 = VertexNormalWS()
  n2 = CollectionParameter() {
   collection=&FortniteMaterialParameters
   parameterName=SunlightDirectionalVector
  }
  n3 = ComponentMask(None=@n2) {b=true, g=true, r=true}
  n4 = VectorParameter() {group=QuantizeOverlay, parameterName=QuantizeSunDirection_Mult, defaultValue=<rgba 1 1 1 1>}
  n5 = StaticSwitchParameter(True=Normalize(VectorInput=Multiply(A=@n3, B=@n4.RGB)), False=@n3) {group=QuantizeOverlay, parameterName=QuantizeOverlay_Modify_SunDirection}
  n6 = DotProduct(A=@n1, B=@n5)
  n7 = ScalarParameter() {defaultValue=0.25, group=QuantizeOverlay, parameterName=Quantize_CenterFlatten, sortPriority=1001}
  n8 = Subtract(A=Abs(None=@n6), B=@n7)
  n9 = Constant() {r=1}
  n10 = Multiply(A=Divide(A=Saturate(None=@n8), B=Subtract(A=@n9, B=@n7)), B=Sign(None=@n6))
  n11 = ScalarParameter() {defaultValue=8, group=QuantizeOverlay, parameterName=Quantize_Steps}
  n12 = Multiply(A=Divide(A=Add(A=@n10)), B=@n11)
  n13 = Max(
   A=Divide(A=Round(None=@n12), B=@n11)
   B=ScalarParameter() {defaultValue=0.25, group=QuantizeOverlay, parameterName=Quantize_MinClamp, sortPriority=1001}
  )
  n14 = ScalarParameter() {defaultValue=0.75, group=QuantizeOverlay, parameterName=Quantize_MaxClamp, sortPriority=1002}
  n15 = Saturate(None=LinearInterpolate(
    A=ScalarParameter() {group=QuantizeOverlay, parameterName=Quantize_MinRangeDarkness, sortPriority=1101}
    B=ScalarParameter() {defaultValue=0.75, group=QuantizeOverlay, parameterName=Quantize_MaxRangeBrigthness, sortPriority=1102}
    Alpha=Min(A=@n13, B=@n14)
   ))
  n16 = Add(A=DotProduct(A=@n1, B=@n5))
  Add(
   A=@n15
   B=Power(Base=Divide(A=@n16), Exp=ScalarParameter() {defaultValue=0.8546, group=QuantizeOverlay, parameterName=Quantize_RemoveSoftGradientPower})
  )
  n17 = FunctionInput() {inputType=FunctionInput_MaterialAttributes}
  n18 = GetMaterialAttributes(None=@n17) {attributeGetTypes [#1]}
  n19 = CameraVectorWS()
  n20 = DotProduct(A=@n18.Normal, B=@n19)
  n21 = ScalarParameter() {defaultValue=3, group=QuantizeOverlay, parameterName=QuantizeNormalSteps}
  n22 = Frac(None=Multiply(A=Abs(None=@n20), B=@n21))
  n23 = Divide(A=@n22, B=@n21)
  Add(A=@n20, B=@n23)
  Arccosine!(None=@n23)
  n24 = CrossProduct(A=@n18.Normal, B=@n19)
  CrossProduct(A=@n18.Normal, B=CrossProduct(A=@n24, B=@n19))
  n25 = VectorParameter() {group=QuantizeOverlay, parameterName=QuantizeNormals_Detail, defaultValue=<rgba 1000 1000 1000 1>}
  Divide(A=Round(None=Multiply(A=@n1, B=@n25.RGB)), B=@n25.RGB)
  n26 = GetMaterialAttributes(None=@n17) {attributeGetTypes [#2, #3]}
  Divide(A=@n26.BaseColor, B=Desaturation(None=@n26.BaseColor))
  Divide(A=@n9, B=@n7)
  n27 = LinearInterpolate(A=@n18.Normal, B=@n19, Alpha=Subtract(A=@n20, B=@n23))
  n28 = StaticSwitchParameter(
   True=SetMaterialAttributes!(MaterialAttributes=@n18.MaterialAttributes, Normal=@n27) {attributeSetTypes [#1]}
   False=@n17
  ) {group=QuantizeOverlay, parameterName=QuantizNormals}
  n29 = GetMaterialAttributes(None=@n28) {attributeGetTypes [#2, #3]}
  n30 = SetMaterialAttributes!(
   MaterialAttributes=@n29.MaterialAttributes
   BaseColor=MaterialFunctionCall(Base=@n29.BaseColor, Blend=@n15) {materialFunction=&Blend_Overlay}.Result
  ) {attributeSetTypes [#2]}
  FunctionOutput(None=MaterialProxyReplace(Realtime=StaticSwitchParameter(True=@n30, False=@n28) {group=QuantizeOverlay, parameterName=QuantizeOveraly}, MaterialProxy=@n17))
  Length(None=@n26.BaseColor) {is=class MaterialExpressionLength}
  n31 = Arccosine(None=@n20)
  MaterialFunctionCall!(Degrees=@n31) {materialFunction=/Engine/Functions/Engine_MaterialFunctions01/Math/DegreesToRadians}
  MaterialFunctionCall!(Radians=@n31) {materialFunction=/Engine/Functions/Engine_MaterialFunctions01/Math/RadiansToDegrees}
  OneMinus(None=@n22)
  PixelNormalWS()
  n32 = Normalize(VectorInput=@n26.BaseColor)
  n33 = LinearInterpolate(
   A=ScalarParameter() {defaultValue=0.22, parameterName=Quantize_Diffuse_Range_Low}
   B=ScalarParameter() {defaultValue=0.528, parameterName=Quantize_Diffuse_Range_High}
   Alpha=Length(None=@n26.BaseColor)
  )
  n34 = ScalarParameter() {defaultValue=4, parameterName=Quantize_Diffuse_Steps}
  n35 = Divide(A=Round(None=Multiply(A=@n33, B=@n34)), B=@n34)
  StaticSwitchParameter(
   True=SetMaterialAttributes(MaterialAttributes=@n26.MaterialAttributes, BaseColor=Multiply(A=@n32, B=@n35)) {attributeSetTypes [#2]}
   False=@n17
  ) {group=PBWAssetStylize, parameterName=PBWAssetStylize}
  n36 = Normalize(VectorInput=@n24)
  n37 = Multiply(A=@n31, B=ScalarParameter() {defaultValue=4.4436, parameterName=TestN})
  n38 = Constant3Vector()
  n39 = Constant3Vector() {constant=<rgba 0 0 1 0>}
  Subtract(B=Add(A=RotateAboutAxis(NormalizedRotationAxis=@n36, RotationAngle=@n37, PivotPoint=@n38, Position=@n39), B=@n39))
  Transform(None=@n27) {transformSourceType=TRANSFORMSOURCE_World, transformType=TRANSFORM_Tangent}
  Transform(None=@n18.Normal)
  VertexNormalWS()
  VertexTangentWS()
 ]}
